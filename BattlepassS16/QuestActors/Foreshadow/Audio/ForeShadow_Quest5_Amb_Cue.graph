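# Ambient SoundCue for the Season 16 Battle Pass Foreshadow Quest 5 actor, in the quest sound class with overridden short-range attenuation.
# A mixer layers a looping ambience bed, a supply drop target loop pitched up by a modulator, and quieter chatter.
# The chatter loops through random picks from five one-shots, each after a random delay.
# An enveloper at the root fades the whole mix in at the start.
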SoundCue /BattlepassS16/QuestActors/Foreshadow/Audio/ForeShadow_Quest5_Amb_Cue.Foreshadow_Quest5_Amb_Cue
root=Enveloper_0
nodes [
 {id=Delay_0, children [Random_0], properties {delayMax=10, delayMin=1}}
 {
  id=Enveloper_0
  children [Mixer_0]
  properties {pitchCurve {editorCurveData {keys [{time=0}]}}, volumeCurve {editorCurveData {keys [<at 0.0045 0.0094>,<at 0.4887 1.0036>]}}}
 }
 {id=Looping_0, children [Delay_0]}
 {id=Mixer_0, children [WavePlayer_0,Modulator_0,Looping_0], properties {inputVolume [1,1,0.5]}}
 {id=Modulator_0, children [WavePlayer_1], properties {pitchMax=2, pitchMin=2, volumeMax=1, volumeMin=1}}
 {id=Random_0, children [WavePlayer_2,WavePlayer_3,WavePlayer_4,WavePlayer_5,WavePlayer_6], properties {weights [1,1,1,1,1]}}
 {id=WavePlayer_0, properties {bLooping=true, soundWaveAssetPtr=&/Foreshadow_Quest5_Amb_Loop_02}}
 {id=WavePlayer_1, properties {bLooping=true, soundWaveAssetPtr=/Game/Sounds/Foley_Loot/Containers/SupplyDrops/sfx_supplydrop_target_loop}}
 {id=WavePlayer_2, properties {soundWaveAssetPtr=&/Foreshadow_Quest5_Amb_Chatter_01}}
 {id=WavePlayer_3, properties {soundWaveAssetPtr=&/Foreshadow_Quest5_Amb_Chatter_02}}
 {id=WavePlayer_4, properties {soundWaveAssetPtr=&/Foreshadow_Quest5_Amb_Chatter_03}}
 {id=WavePlayer_5, properties {soundWaveAssetPtr=&/Foreshadow_Quest5_Amb_Chatter_04}}
 {id=WavePlayer_6, properties {soundWaveAssetPtr=&/Foreshadow_Quest5_Amb_Chatter_05}}
]
properties {
 attenuationSettings=/Game/Sounds/Foley_Loot/Containers/SupplyDrops/Fort_SupplyDrop_Marker_Close_Athena_Att
 bOverrideAttenuation=true
 maxDistance=800
 soundClassObject=&Fort_Quests
 volumeMultiplier=1
 attenuationOverrides {falloffDistance=700, stereoSpread=500, attenuationShapeExtents {x=100}}
}
