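# Sound cue for electrical sparking on a Battlepass S16 Foreshadow quest actor, in the quest sound class with short-range overridden attenuation.
# Two quality levels use the full mix: two delayed loops of pitched-up random fizzy and electric-trap sparks, a looping shockwave-rifle spark bed, and a one-shot spark.
# The other two levels drop the random loops, keeping only the one-shot and the looping bed.
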SoundCue /BattlepassS16/QuestActors/Foreshadow/Audio/ForeShadow_Quest5_Sparks_Cue
root=QualityLevel_0
nodes [
 {id=Delay_1, children [Modulator_4], properties {delayMax=2}}
 {id=Delay_2, children [Modulator_2], properties {delayMax=2}}
 {id=Looping_1, children [Delay_2]}
 {id=Looping_3, children [Delay_1]}
 {id=Mixer_1, children [Looping_1, Looping_3, Modulator_5, WavePlayer_21], properties {inputVolume [0.8,1,0.3,1]}}
 {id=Mixer_2, children [WavePlayer_11], properties {inputVolume [0.35]}}
 {id=Mixer_3, children [WavePlayer_22, Modulator_6], properties {inputVolume [0.5,0.3]}}
 {id=Modulator_2, children [Random_1], properties {pitchMax=1.9, pitchMin=1.25, volumeMin=0.5}}
 {id=Modulator_4, children [Random_6], properties {pitchMax=1.6, pitchMin=1.25}}
 {id=Modulator_5, children [WavePlayer_17], properties {pitchMax=1.3, pitchMin=1.2, volumeMax=1, volumeMin=1}}
 {id=Modulator_6, children [WavePlayer_23], properties {pitchMax=1.3, pitchMin=1.2, volumeMax=1, volumeMin=1}}
 {id=QualityLevel_0, children [Mixer_1,Mixer_1,Mixer_3,Mixer_3]}
 {id=Random_1, children [WavePlayer_5,WavePlayer_6,WavePlayer_7], properties {weights [1,1,1]}}
 {id=Random_6, children [Mixer_2, WavePlayer_19, WavePlayer_20], properties {weights [1,1,1]}}
 {id=WavePlayer_11, properties {soundWaveAssetPtr=/Game/Sounds/Fort_Traps/Electric_Trap/CeilingTrap_Electric_Spark_007}}
 {
  id=WavePlayer_17
  properties {bLooping=true, soundWaveAssetPtr=/Game/Sounds/Fort_Weapons/Rifle/Shockwave/Weapon_Shockwave_AR_Muzzle_Sparks_Prox_Loop}
 }
 {id=WavePlayer_19, properties {soundWaveAssetPtr=/Game/Sounds/Fort_Traps/Electric_Trap/CeilingTrap_Electric_Spark_mono_006}}
 {id=WavePlayer_20, properties {soundWaveAssetPtr=/Game/Sounds/Fort_Traps/Electric_Trap/CeilingTrap_Electric_Spark_mono_002}}
 {id=WavePlayer_21, properties {soundWaveAssetPtr=&CeilingTrap_Electric_Spark_007}}
 {id=WavePlayer_22, properties {soundWaveAssetPtr=&CeilingTrap_Electric_Spark_007}}
 {
  id=WavePlayer_23
  properties {bLooping=true, soundWaveAssetPtr=&Weapon_Shockwave_AR_Muzzle_Sparks_Prox_Loop}
 }
 {id=WavePlayer_5, properties {soundWaveAssetPtr=/Game/Sounds/Fort_Weapons/BlackMetal/SparkBursts/BlackMetal_Sparks_Fizzy_01}}
 {id=WavePlayer_6, properties {soundWaveAssetPtr=/Game/Sounds/Fort_Weapons/BlackMetal/SparkBursts/BlackMetal_Sparks_Fizzy_04}}
 {id=WavePlayer_7, properties {soundWaveAssetPtr=/Game/Sounds/Fort_Weapons/BlackMetal/SparkBursts/BlackMetal_Sparks_Fizzy_05}}
]
properties {
 attenuationSettings=/Game/Sounds/Foley_Loot/Containers/SupplyDrops/Fort_SupplyDrop_Marker_Close_Athena_Att
 bOverrideAttenuation=true
 maxDistance=800
 soundClassObject=&Fort_Quests
 volumeMultiplier=1.3
 attenuationOverrides {falloffDistance=700, stereoSpread=500, attenuationShapeExtents {x=100}}
}
