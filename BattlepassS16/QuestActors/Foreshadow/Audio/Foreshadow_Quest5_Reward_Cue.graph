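# Sound cue for the reward moment of the Season 16 battle pass Foreshadow Quest 5, routed to the UI sound class.
# A mixer layers a reward-screen chest loot upgrade sting with wood, workbench junk and computer impact hits at individually weighted volumes under a reduced overall gain.
# Concurrency is overridden to one instance, stopping the oldest when retriggered.
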SoundCue /BattlepassS16/QuestActors/Foreshadow/Audio/Foreshadow_Quest5_Reward_Cue
root=Mixer_0
nodes [
 {id=Mixer_0, children [WavePlayer_1,WavePlayer_5,WavePlayer_4,WavePlayer_6], properties {inputVolume [1,1.2,0.9,0.6]}}
 {id=WavePlayer_1, properties {soundWaveAssetPtr=/Game/Sounds/Fort_UI_Sounds/Rewards/Fort_RewardScreen_TeamScore_ChestLootUpgrade}}
 {id=WavePlayer_4, properties {soundWaveAssetPtr=/Game/Sounds/Fort_Impact_Sounds/Demolition/Wood/Wood_Small_Impact_mix_01}}
 {id=WavePlayer_5, properties {soundWaveAssetPtr=/Game/Sounds/Fort_Impact_Sounds/Demolition/Workbench/Workbench_Junk_Hit3}}
 {id=WavePlayer_6, properties {soundWaveAssetPtr=/Game/Sounds/Fort_Impact_Sounds/Demolition/Computer/Computer_Impact_01}}
]
properties {
 bOverrideConcurrency=true
 maxDistance=2097152
 soundClassObject=&Fort_UI
 volumeMultiplier=0.4
 concurrencyOverrides {maxCount=1, resolutionRule=StopOldest}
}
