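; Season 16 foreshadowing quest radio actor that plays staged Black Helicopter voice-over dialogue.
; ConversationMessage passes the message number to PlayingRadio, prints a debug string, then calls its parent.
; PlayingRadio acts only on the non-authority (remote) side, stopping and reusing the ambient audio component for playback.
; An integer switch async-loads and assigns one of five stage dialogue cues, with message zero assigning none.
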(event EventBeginPlay
  (|CallParentFunction))

(event Collision|EventActorBeginOverlap
  (params (OtherActor :type Actor%20Object%20Reference))
  (|CallParentFunction OtherActor) ; This node is disabled and will not be called.\nDrag off pins to build functionality.
)

(event EventTick
  (params (DeltaSeconds :type Float%20%28single-precision%29))
  (|CallParentFunction DeltaSeconds) ; This node is disabled and will not be called.\nDrag off pins to build functionality.
)

(event Custom|PlayingRadio
  (params (Message :type Integer))
  (Networking|SwitchHasAuthority
    (:Remote
      (Audio|Components|Audio|Stop (Variables|Components|GetAmbientAudio))
      (switch int Message
        (:0
          _)
        (:1
          (Utilities|AsyncLoadAsset "/BattlepassS16/Sounds/VO/Foreshadowing/Quest_2/S16_2021_BRQuest_BlackHelicopter_Stage1_Dialogue_Cue.S16_2021_BRQuest_BlackHelicopter_Stage1_Dialogue_Cue"
            (:Completed
              (Audio|Components|Audio|SetSound (Variables|Components|GetAmbientAudio) "/BattlepassS16/Sounds/VO/Foreshadowing/Quest_2/S16_2021_BRQuest_BlackHelicopter_Stage1_Dialogue_Cue.S16_2021_BRQuest_BlackHelicopter_Stage1_Dialogue_Cue"))))
        (:2
          (Utilities|AsyncLoadAsset "/BattlepassS16/Sounds/VO/Foreshadowing/Quest_2/S16_2021_BRQuest_BlackHelicopter_Stage2_Dialogue_Cue.S16_2021_BRQuest_BlackHelicopter_Stage2_Dialogue_Cue"
            (:Completed
              (Audio|Components|Audio|SetSound (Variables|Components|GetAmbientAudio) "/BattlepassS16/Sounds/VO/Foreshadowing/Quest_2/S16_2021_BRQuest_BlackHelicopter_Stage2_Dialogue_Cue.S16_2021_BRQuest_BlackHelicopter_Stage2_Dialogue_Cue"))))
        (:3
          (Utilities|AsyncLoadAsset "/BattlepassS16/Sounds/VO/Foreshadowing/Quest_2/S16_2021_BRQuest_BlackHelicopter_Stage3_Dialogue_Cue.S16_2021_BRQuest_BlackHelicopter_Stage3_Dialogue_Cue"
            (:Completed
              (Audio|Components|Audio|SetSound (Variables|Components|GetAmbientAudio) "/BattlepassS16/Sounds/VO/Foreshadowing/Quest_2/S16_2021_BRQuest_BlackHelicopter_Stage3_Dialogue_Cue.S16_2021_BRQuest_BlackHelicopter_Stage3_Dialogue_Cue"))))
        (:4
          (Utilities|AsyncLoadAsset "/BattlepassS16/Sounds/VO/Foreshadowing/Quest_2/S16_2021_BRQuest_BlackHelicopter_Stage4_Dialogue_Cue.S16_2021_BRQuest_BlackHelicopter_Stage4_Dialogue_Cue"
            (:Completed
              (Audio|Components|Audio|SetSound (Variables|Components|GetAmbientAudio) "/BattlepassS16/Sounds/VO/Foreshadowing/Quest_2/S16_2021_BRQuest_BlackHelicopter_Stage4_Dialogue_Cue.S16_2021_BRQuest_BlackHelicopter_Stage4_Dialogue_Cue"))))
        (:5
          (Utilities|AsyncLoadAsset "/BattlepassS16/Sounds/VO/Foreshadowing/Quest_2/S16_2021_BRQuest_BlackHelicopter_Stage5_Dialogue_Cue.S16_2021_BRQuest_BlackHelicopter_Stage5_Dialogue_Cue"
            (:Completed
              (Audio|Components|Audio|SetSound (Variables|Components|GetAmbientAudio) "/BattlepassS16/Sounds/VO/Foreshadowing/Quest_2/S16_2021_BRQuest_BlackHelicopter_Stage5_Dialogue_Cue.S16_2021_BRQuest_BlackHelicopter_Stage5_Dialogue_Cue")))))
      (Audio|Components|Audio|Play (Variables|Components|GetAmbientAudio) 0.0))))

(fn ConversationMessage
  (params (MessageNumber :type Integer))
  (CallFunction|PlayingRadio MessageNumber)
  (Development|PrintString (Utilities|String|Append "MessageShouldPlay
" (Utilities|String|ToString(Integer) MessageNumber)))
  (|CallParentFunction MessageNumber))

(fn ConstructionScript
  (|CallParentFunction))
